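# Post-process style body-part AnimBlueprint for the M_MED_Assemble_P player cosmetic, layering Control Rigs over the pose copied from the attached parent mesh.
# Skydiving uses its own rig; emotes and certain swim states use the base rig; otherwise an optional crouch rig feeds base, then weapon rigs.
# Each update sets a material parameter hiding element 09 when hands are replaced, except during emotes, skydiving or certain driving.
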AnimBlueprint /BRCosmetics/Characters/Player/Male/Medium/Bodies/M_MED_Assemble_P/Meshes/M_MED_Assemble_P_Skeleton_AnimBlueprint
target_skeleton=&/M_MED_Assemble_P_Skeleton
graphs [
 AnimGraph {nodes [
   "Constraint - dyn_lowerarm_l (R)"(Alpha=1, ConstraintWeights_0=1) {
    alphaBoolBlend {blendOption=Linear}
    boneToModify {boneName=dyn_lowerarm_l}
    constraintSetup [{transformType=Rotation, targetBone {boneName=lowerarm_l}}]
    constraintWeights [1]
   }
   "Constraint - dyn_lowerarm_r (R)"(Alpha=1, ConstraintWeights_0=1) {
    alphaBoolBlend {blendOption=Linear}
    boneToModify {boneName=dyn_lowerarm_r}
    constraintSetup [{transformType=Rotation, targetBone {boneName=lowerarm_r}}]
    constraintWeights [1]
   }
   n1 = Control Rig(Source=Use cached pose 'SavedPose8850'()) {
    alphaBoolBlend {blendOption=Linear}
    controlRigAssetReference {blueprintRigClass=&/M_MED_Assemble_P_CtrlRig.M_MED_Assemble_P_CtrlRig_C}
   }
   n2 = Control Rig(Source=Use cached pose 'SavedPose8850'()) {
    alphaBoolBlend {blendOption=Linear}
    controlRigAssetReference {blueprintRigClass=&/M_MED_Assemble_P_Crouch_CtrlRig.M_MED_Assemble_P_Crouch_CtrlRig_C}
   }
   n3 = Control Rig(Source=Blend Poses by bool(BlendPose_0=@n2, BlendPose_1=Use cached pose 'SavedPose8850'(), bActiveValue=Get bIsCrouching()) {blendType=Linear}) {
    alphaBoolBlend {blendOption=Linear}
    controlRigAssetReference {blueprintRigClass=&M_MED_Assemble_P_CtrlRig_C}
   }
   n4 = Control Rig(Source=@n3) {
    alphaBoolBlend {blendOption=Linear}
    controlRigAssetReference {blueprintRigClass=&/M_MED_Assemble_P_Weapon_CtrlRig.M_MED_Assemble_P_Weapon_CtrlRig_C}
   }
   n5 = OR Boolean!(
    A=Get bIsPlayingEmote()
    B=Get bIsSurfaceSwimming()
    C:bool=Get bIsSwimSprinting() -> (bIsSwimSprinting:bool)
    D:bool=Get bIsWaterJump() -> (bIsWaterJump:bool)
   ) -> (ReturnValue)
   n6 = Control Rig(Source=Use cached pose 'SavedPose8850'()) {
    alphaBoolBlend {blendOption=Linear}
    controlRigAssetReference {blueprintRigClass=&/M_MED_Assemble_P_Skydive_CtrlRig.M_MED_Assemble_P_Skydive_CtrlRig_C}
   }
   Output Pose(Result=Blend Poses by bool(
     BlendPose_0=@n6
     BlendPose_1=Blend Poses by bool(BlendPose_0=@n1, BlendPose_1=@n4, bActiveValue=@n5) {blendType=Linear}
     bActiveValue=Get bIsSkyDiving()
    ) {blendType=Linear}) {layerGroup=None}
   Save cached pose 'SavedPose8850'(Pose=Copy Pose From Mesh() {bUseAttachedParent=true}) {cacheName=SavedPose8850}
  ]}
 EventGraph {nodes [
   Event BlueprintUpdateAnimation()
   > n1 = Sequence()
   > n2 = Cast To FortPlayerPawn(Object=GetOwningActor())
   > Set As Fort Player Pawn(As Fort Player Pawn:Fort Player Pawn Object Reference=@n2) -> (Output_Get:Fort Player Pawn Object Reference)
   > n3 = Get!(execute:Exec) -> (then:Exec, else:Exec, As Fort Player Pawn:Fort Player Pawn Object Reference)
   > Set is_wearing_westsausage(is_wearing_westsausage:bool=HasMatchingGameplayTag!(self:Gameplay Tag Asset Interface Interface=@n3, TagToCheck:Gameplay Tag Structure=(TagName="Granted.Athena.HandsAreReplaced")) -> (ReturnValue:bool)) -> (Output_Get:bool)
   n4 = Get Dynamic Material Instance(execute=@n1.then_1, SkeletalMeshComponent=GetPartSkeletalMeshComponent())
   > n5 = Branch(Condition=Get is_wearing_westsausage() -> (is_wearing_westsausage:bool))
   > n6 = Branch(Condition=Get is_performing_emote() -> (is_performing_emote:bool))
   SetScalarParameterValue(execute=[@n10.then, @n9.then, @n6.then, @n8.then, @n5.else, @n7.then], self=@n4, ParameterName=Hide Element 09, Value=0)
   n7 = Branch(execute=@n6.else, Condition=Get bIsValetDriver() -> (bIsValetDriver:bool))
   n8 = Branch(execute=@n7.else, Condition=Get bIsMeatballDriver() -> (bIsMeatballDriver:bool))
   n9 = Branch(execute=@n8.else, Condition=Get bIsAntelopeVehicleDriver() -> (bIsAntelopeVehicleDriver:bool))
   n10 = Branch(execute=@n9.else, Condition=Get bIsSkyDiving())
   SetScalarParameterValue(execute=@n10.else, self=@n4, ParameterName=Hide Element 09, Value=1)
   Event OnBeginPlayingEmote() {is=title K2Node_Event}
   > Set is_performing_emote(is_performing_emote:bool=true) -> (Output_Get:bool)
   Event OnEndPlayingEmote() {is=title K2Node_Event}
   > Set is_performing_emote(is_performing_emote=false)
  ]}
]
variables [
 As Fort Player Pawn {type=FortPlayerPawn}
 is_performing_emote {type=boolean}
 is_wearing_westsausage {type=boolean}
]
