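# Physics asset for the face accessory mesh of the male medium Assemble_P player cosmetic body, simulating two small dangling parts on the head.
# A sphere body covers the head, and capsule bodies cover the dyn_handle and dyn_meter bones.
# Each capsule is constrained to the head with a narrow angular range that leaves one main swing axis free, so the handle and meter each pivot in a hinge-like motion.
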PhysicsAsset /BRCosmetics/Characters/Player/Male/Medium/Bodies/M_MED_Assemble_P/Meshes/Parts/M_MED_Assemble_P_FaceAcc_Physics
bodies {
 dyn_handle [{shapeType=Capsule, length=6, radius=2.5, shapeName=dyn_handle_capsule, center=<0.6134 2.8198 0.0517>, rotation=<rot -1.3868 -89.9999 1.5977>}]
 dyn_meter [{
   shapeType=Capsule
   length=2.5905
   radius=0.9505
   shapeName=dyn_meter_capsule
   center=<-0.1713 0.009 0.1362>
   rotation=<rot 0 89.5206 -89.5362>
  }]
 head [{shapeType=Sphere, radius=12, shapeName=head_sphere, center=<9.3768 2.1369 0>}]
}
constraints [
 {bone1Name=dyn_handle, bone2Name=head, swing1LimitDegrees=25, swing2LimitDegrees=1, twistLimitDegrees=1}
 {bone1Name=dyn_meter, bone2Name=head, swing1LimitDegrees=1, swing2LimitDegrees=40, twistLimitDegrees=1}
]
